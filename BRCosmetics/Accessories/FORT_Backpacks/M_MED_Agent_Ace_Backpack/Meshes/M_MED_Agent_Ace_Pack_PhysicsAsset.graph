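# Physics asset for the Agent Ace backpack cosmetic, defining secondary motion for its two dangling ammo pieces and their bullets.
# Each side forms a shallow capsule hierarchy: a small master body anchors an ammo body, which carries three bullet bodies.
# Tight swing and twist limits keep movement to a slight jiggle.
# Bullet side names are crossed: right bullets constrain to the left ammo body, and left bullets to the right.
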PhysicsAsset /BRCosmetics/Accessories/FORT_Backpacks/M_MED_Agent_Ace_Backpack/Meshes/M_MED_Agent_Ace_Pack_PhysicsAsset
bodies {
 dyn_ammo_l [{shapeType=Capsule, length=4, radius=2, center=<2 0 0>, rotation=<rot 69.9999 -180 179.9999>}]
 dyn_ammo_mstr_l [{shapeType=Capsule, length=1, radius=1, rotation=<rot 0 0 90>}]
 dyn_ammo_mstr_r [{shapeType=Capsule, length=1, radius=1, rotation=<rot 0 0 90>}]
 dyn_ammo_r [{shapeType=Capsule, length=4, radius=2, center=<2 0 0>, rotation=<rot 90 0 0>}]
 dyn_bullet_l_01 [{shapeType=Capsule, length=4, radius=1.1, center=<0 0 1.3>, rotation=<rot 90 0 0>}]
 dyn_bullet_l_02 [{shapeType=Capsule, length=4, radius=1.1, center=<0 0 1.3>, rotation=<rot 90 0 0>}]
 dyn_bullet_l_03 [{shapeType=Capsule, length=4, radius=1.1, center=<0 0 1.3>, rotation=<rot 90 0 0>}]
 dyn_bullet_r_01 [{shapeType=Capsule, length=4, radius=1.2, center=<0 0 1.3>, rotation=<rot 90 0 0>}]
 dyn_bullet_r_02 [{shapeType=Capsule, length=4, radius=1.1, center=<0 0 1.3>, rotation=<rot 90 0 0>}]
 dyn_bullet_r_03 [{shapeType=Capsule, length=4, radius=1.1, center=<0 0 1.3>, rotation=<rot 90 0 0>}]
}
constraints [
 {bone1Name=dyn_bullet_r_01, bone2Name=dyn_ammo_l, swing1LimitDegrees=5, swing2LimitDegrees=10, twistLimitDegrees=5}
 {bone1Name=dyn_bullet_r_02, bone2Name=dyn_ammo_l, swing1LimitDegrees=5, swing2LimitDegrees=10, twistLimitDegrees=5}
 {bone1Name=dyn_bullet_r_03, bone2Name=dyn_ammo_l, swing1LimitDegrees=5, swing2LimitDegrees=10, twistLimitDegrees=5}
 {bone1Name=dyn_bullet_l_01, bone2Name=dyn_ammo_r, swing1LimitDegrees=5, swing2LimitDegrees=10, twistLimitDegrees=5}
 {bone1Name=dyn_bullet_l_02, bone2Name=dyn_ammo_r, swing1LimitDegrees=5, swing2LimitDegrees=10, twistLimitDegrees=5}
 {bone1Name=dyn_bullet_l_03, bone2Name=dyn_ammo_r, swing1LimitDegrees=5, swing2LimitDegrees=10, twistLimitDegrees=5}
 {bone1Name=dyn_ammo_l, bone2Name=dyn_ammo_mstr_l, swing1LimitDegrees=8, swing2LimitDegrees=5, twistLimitDegrees=2}
 {bone1Name=dyn_ammo_r, bone2Name=dyn_ammo_mstr_r, swing1LimitDegrees=8, swing2LimitDegrees=5, twistLimitDegrees=2}
]
